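; Construction script for the fire escape walkway stairs building actor blueprint.
; It fetches a hit sound and a destroyed sound through audio getters and assigns them as the actor's play-hit and death sounds.
; The script does nothing else, so any other stair behavior lies outside it.
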(fn ConstructionScript
  (Variables|Effects|SetPlayHitSound (Variables|Audio|GetHit%20Sound))
  (Variables|Effects|SetDeathSound (Variables|Audio|GetDestroyed%20Sound)))
